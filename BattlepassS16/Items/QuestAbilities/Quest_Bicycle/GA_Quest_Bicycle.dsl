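; GA_Quest_Bicycle is a quest gameplay ability that applies a configured gameplay effect to the owning player when an objective of the Bicycle quest completes.
; On activation it compares the pawn's gameplay tags against T_Bicycle, then either binds to the player controller's quest objective state change event or ends immediately.
; When the callback reports a completed objective of Quest_Bicycle, it applies GE_Bicycle to the pawn, unbinds and ends.
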(event Ability|EventActivateAbility
  (|CallParentFunction)
  (if (GameplayTags|TagContainer|CompareTagAssetInterfacetoOtherTags (Variables|GAVars|Setup|GetT_Bicycle) (Variables|GAVars|Data|GetOwningPlayerPawn))
    (EventDispatchers|BindEventtoOnQuestObjectiveStateChanged (Fortnite|GetFortPlayerControllerfromActor (Variables|GAVars|Data|GetOwningPlayerPawn)) (delegate Custom|Check%20Quest%20Complete))
    (else
      (Ability|EndAbility))))

(event Ability|EventFailedToActivatePassiveAbility
  (params (ActorInfo :type Gameplay%20Ability%20Actor%20Info%20Structure)))

(event Custom|Check%20Quest%20Complete
  (params (Controller :type Fort%20Player%20Controller%20Object%20Reference) (QuestDef :type Fort%20Quest%20Item%20Definition%20Object%20Reference) (BackendName :type Name) (CompletionCount :type Integer) (ObjectiveCompleted :type Boolean) (QuestCompleted :type Boolean))
  (if (Utilities|Equal(Object) QuestDef (Variables|GAVars|Setup|GetQuest_Bicycle))
    (if ObjectiveCompleted
      (GameplayEffects|ApplyGameplayEffectToSelf :self (Ability|GetAbilitySystemComponent (Variables|GAVars|Data|GetOwningPlayerPawn)) :GameplayEffectClass (Variables|GAVars|Setup|GetGE_Bicycle) :Level 0.0)
      (EventDispatchers|UnbindEventfromOnQuestObjectiveStateChanged (Fortnite|GetFortPlayerControllerfromActor (Variables|GAVars|Data|GetOwningPlayerPawn)) (delegate Custom|Check%20Quest%20Complete))
      (Ability|EndAbility))))
